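; Anim notify state for the Kepler player character that plays a Niagara drool effect during emote animations.
; Begin attaches the effect to the pawn's Face mesh part at a configurable bone and sets a bool parameter flagging frontend preview display.
; End stores no component reference; it scans the Face mesh's child components and deactivates Niagara components matching that system and attach socket.
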(fn Received_NotifyBegin
  (params (MeshComp :type Skeletal%20Mesh%20Component%20Object%20Reference) (Animation :type Anim%20Sequence%20Base%20Object%20Reference) (TotalDuration :type Float%20%28single-precision%29) (EventReference :type Anim%20Notify%20Event%20Reference%20Structure%20%28by%20ref%29))
  (bind _asfort_player_pawn_athena (Utilities|Casting|CastToFortPlayerPawnAthena (Components|GetOwner MeshComp)))
  (bind _returnvalue (Niagara|SpawnSystemAttached (Variables|Default|GetDroolEffect) (Customization|GetSkeletalMeshforPartType _asfort_player_pawn_athena "Face" :!class /Script/FortniteGame.FortPlayerPawn) (Variables|Default|GetAttachBoneName) "0, 0, 0" "0, 0, 0" "SnapToTarget" true))
  (Effects|Components|ParticleSystem|SetBoolParameter _returnvalue "bIsFrontEndPreview" (Reactive|Frontend|IsCosmeticObjectDisplayedinFrontendContext (Customization|GetSkeletalMeshforPartType _asfort_player_pawn_athena "Face" :!class /Script/FortniteGame.FortPlayerPawn)))
  (return false))

(fn Received_NotifyEnd
  (params (MeshComp :type Skeletal%20Mesh%20Component%20Object%20Reference) (Animation :type Anim%20Sequence%20Base%20Object%20Reference) (EventReference :type Anim%20Notify%20Event%20Reference%20Structure%20%28by%20ref%29))
  (bind _asfort_player_pawn_athena (Utilities|Casting|CastToFortPlayerPawnAthena (Components|GetOwner MeshComp)))
  (if (Utilities|IsValid (Customization|GetSkeletalMeshforPartType _asfort_player_pawn_athena "Face" :!class /Script/FortniteGame.FortPlayerPawn) :!class /Script/Engine.KismetSystemLibrary)
    (for _array_element (Components|GetChildrenComponents (Customization|GetSkeletalMeshforPartType _asfort_player_pawn_athena "Face" :!class /Script/FortniteGame.FortPlayerPawn) false)
      (bind _asniagara_particle_system_component (Utilities|Casting|CastToNiagaraComponent _array_element))
      (if (and (Utilities|Equal(Object) (Effects|Components|ParticleSystem|GetFXSystemAsset _asniagara_particle_system_component) (Variables|Default|GetDroolEffect)) (Utilities|Name|Equal(Name) (Transformation|GetAttachSocketName _asniagara_particle_system_component) (Variables|Default|GetAttachBoneName)))
        (Components|Activation|Deactivate _asniagara_particle_system_component)))
    (return false)))
