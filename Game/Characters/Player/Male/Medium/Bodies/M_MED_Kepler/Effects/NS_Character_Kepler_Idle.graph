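# Niagara idle effect for the Kepler player character that drips drool from skeletal mesh sample points, suppressing new strands when the head is hidden.
# DroolMain spawns pendulum-driven mesh strands stretched between anchor and tip, with ReactiveRate adding more.
# Its WriteDeath module feeds positions through an array to system values that spawn falling DroolDrips.
# JumpDrool throws short-lived stretched strands under gravity while SpawnBurst is set.
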NiagaraSystem /Game/Characters/Player/Male/Medium/Bodies/M_MED_Kepler/Effects/NS_Character_Kepler_Idle
dependencies {
 usedDataInterfaces [
  &NiagaraDataInterfaceCurve
  &NiagaraDataInterfaceMeshRendererInfo
  &NiagaraDataInterfaceSkeletalMesh
  &NiagaraDataInterfaceArrayPosition
 ]
 usedDynamicInputs []
 usedModules [
  &self:SystemSpawnScript.SystemScriptSource.SystemScriptGraph.NiagaraNodeAssignment_0.SetVariables_#1
  &self:SystemSpawnScript.SystemScriptSource.SystemScriptGraph.NiagaraNodeAssignment_1.SetVariables_#2
  &SystemState
  &self:ReadDeath
  &self:SystemSpawnScript.SystemScriptSource.SystemScriptGraph.NiagaraNodeAssignment_2.SetVariables_#3
 ]
}
emitters {
 DroolDrips {
  emitterUpdateScript [
   &EmitterState()
   &SpawnPerFrame(Spawn:NiagaraBool=$System.DeathSpawn, Use Spawn Probability=false)
  ]
  particleSpawnScript [
   &InitializeParticle(
    Lifetime=&UniformRangedFloat(Minimum=0.25, Maximum=0.5)
    Mesh Scale=&Multiply_VectorByFloat(Vector=<6 0.35 0.35>, Float=&RandomRangeFloat(Minimum=0.95, Maximum=1.05))
    Mesh Scale Mode=::NewEnumerator4
    Position=$System.DeathPosition
    Position Mode=::NewEnumerator1
    Sprite Size Mode=::NewEnumerator3
    Uniform Sprite Size=5
    UsePositionOffset=false
    Write Material Random=true
   )
   &AddVelocity(Velocity=<0 0 -5>) {enabled=false}
  ]
  particleUpdateScript [
   &ParticleState()
   &GravityForce()
   &InheritVelocity(
    Inherited Velocity Amount Scale=&VectorFromFloat(Value=-0.1)
    Inherited Velocity Speed Limit=10
    Speed Limit=true
    UseSpeedThreshold=false
   ) {enabled=false}
   &DynamicMaterialParameters(
    Index 0 Param 1=&RandomRangeFloat(Minimum=0, Maximum=1)
    Index 0 Param 2=1
    Module.Index0Float0Write=true
    Module.Index0Float1Write=true
    Module.Index0Float2Write=true
    Module.Index0Float3Write=true
   )
   &ScaleMeshSize(Scale Factor=&Multiply_VectorByFloat(
     Vector=&MakeVector(
      X=&Lerp_Float(A=0.25, B=1, Alpha=$Particles.NormalizedAge)
      Y=&Lerp_Float(A=1, B=0.5, Alpha=$Particles.NormalizedAge)
      Z=&Lerp_Float(A=1, B=0.5, Alpha=$Particles.NormalizedAge)
     )
     Float=&Lerp_Float(A=1, B=0, Alpha=$Particles.NormalizedAge)
    ))
   &Drag(Drag=2, Use Linear Drag=true, Use Rotational Drag=true)
   &SolveForcesAndVelocity(Clamp Velocity=false, Limit Acceleration=false)
  ]
  rendererClasses [&NiagaraSpriteRendererProperties, &NiagaraMeshRendererProperties]
  renderers [
   {rendererClass=&NiagaraSpriteRendererProperties}
   {rendererClass=&NiagaraMeshRendererProperties, rendererIndex=1}
  ]
 }
 DroolMain {
  emitterUpdateScript [
   &EmitterState()
   &SpawnRate(
    Spawn Probability=&Multiply_Float(A=0.15, B=&MakeCustomFloatFromBool(Bool=$User.bIsHeadHidden, True Float=0, False Float=1))
    SpawnRate=&RandomRangeFloat(Minimum=2, Maximum=5, Evaluation Type=::NewEnumerator1)
    Use Spawn Probability=true
   )
   &SpawnRate(Spawn Probability=0.25, SpawnRate=&Lerp_Float(A=0, B=100, Alpha=$User.ReactiveRate), Use Spawn Probability=true) {module=SpawnRate001}
  ]
  particleSpawnScript [
   &ShapeLocation(
    Handle Radius=0
    Large Radius=2.75
    MaskSpawnGroup=false
    Override Seed=false
    Rotation Coordinate Space=::World
    Rotation Mode=::NewEnumerator1
    Shape Origin=&Add_Vector(A=&FN_FrontEndVector3Modifier(Gameplay=<-5 9.3 2.9>, Front End=<-4 9.2 2.9>), B=<0 0 0>)
    Shape Primitive=::NewEnumerator3
    UseOwnerScale=true
   )
   /SetVariables(
    Particles.BoneOffset:Vector3f=&RandomRangeVector(Minimum=<0 6 -3>, Maximum=<0 4.5 -2>)
    Particles.TorusPosition:NiagaraPosition=$Particles.ShapeLocation.ShapePosition
   ) {isSetParameters=true}
   &SkeletalMeshLocation(
    Bone Sampling Mode=::NewEnumerator4
    MaskSpawnGroup=false
    Orientation Sampling=::NewEnumerator1
    Sampled Position Offset=&AddVectorToPosition(Position=$Particles.TorusPosition, Vector=$Particles.BoneOffset)
    Skeletal Mesh=$System.Skeletal Mesh
   )
   &KillParticlesInVolume(
    Box Size:Vector3f=<5 6 10>
    Kill Shape=::NewEnumerator1
    Origin Offset:Vector3f=<0 3 3>
    Volume Origin:NiagaraPosition=$Output.SkeletalMeshLocation.SampledPositionWithoutOffset
   )
   &InitializeParticle(
    Color=&FNTimeOfDayColor(Night Color=<rgba 0.6 0.6 0.6 1>, Morning Color=<rgba 1 1 1 1>, Day Color=<rgba 1 1 1 1>, Evening Color=<rgba 1 1 1 1>)
    Lifetime=&Lerp_Float(
     A=&UniformRangedFloat(Minimum=0.5, Maximum=1)
     B=&UniformRangedFloat(Minimum=0.5, Maximum=1)
     Alpha=$User.ReactiveRate
    )
    Mesh Scale=<0.09 0.09 0.018>
    Mesh Scale Mode=::NewEnumerator4
    Sprite Size Mode=::NewEnumerator3
    Uniform Sprite Size=2
    UsePositionOffset=false
    Write Material Random=true
   )
   &CameraOffset(Camera Offset Amount=1)
  ]
  particleUpdateScript [
   &ParticleState()
   /SetVariables(Particles.OrigPos:NiagaraPosition=$Particles.Position) {isSetParameters=true}
   &SkeletalMeshLocation(
    Bone Sampling Mode=::NewEnumerator4
    MaskSpawnGroup=false
    Orientation Sampling=::NewEnumerator1
    Sampled Position Offset=&AddVectorToPosition(Position=$Particles.TorusPosition, Vector=$Particles.BoneOffset)
    Skeletal Mesh=$System.Skeletal Mesh
   ) {module=SkeletalMeshLocation001}
   /SetVariables(Particles.MeshPosition:NiagaraPosition=$Particles.Position) {isSetParameters=true}
   /SetVariables(Particles.Position=$Particles.OrigPos) {isSetParameters=true}
   /SetVariables(
    Particles.AlphaScale:NiagaraFloat=&FloatFromCurve(FloatCurve=DataInterface() {propertyValues {Curve {keys [
         {arriveTangent=-0.0275, interpMode=RCIM_Cubic, leaveTangent=-0.0275, tangentMode=RCTM_User, time=0.95, value=1}
         {arriveTangent=-34.2468, interpMode=RCIM_Cubic, leaveTangent=-34.2466, tangentMode=RCTM_User, time=1, value=0}
        ]}}})
    Particles.XYScale:NiagaraFloat=&FloatFromCurve(
     FloatCurve=DataInterface() {propertyValues {Curve {keys [
         {arriveTangent=6.0986, interpMode=RCIM_Cubic, leaveTangent=6.0986, tangentMode=RCTM_User, time=0, value=0}
         {arriveTangent=-1.5439, interpMode=RCIM_Cubic, leaveTangent=-1.5439, tangentMode=RCTM_User, time=0.25, value=1}
         {arriveTangent=-0.3409, interpMode=RCIM_Cubic, leaveTangent=-0.3409, tangentMode=RCTM_User, time=0.6, value=0.6082}
         {arriveTangent=-2.6194, interpMode=RCIM_Cubic, leaveTangent=-2.6194, tangentMode=RCTM_User, time=0.9823, value=0.214}
         <at 1 0.05>
        ]}}}
     Scale Curve=&RandomRangeFloat(Minimum=0.4, Maximum=1)
    )
    Particles.ZScale:NiagaraFloat=&FloatFromCurve(
     FloatCurve=DataInterface() {propertyValues {Curve {keys [
         {arriveTangent=3.0916, interpMode=RCIM_Cubic, leaveTangent=3.0916, tangentMode=RCTM_User, time=0, value=0.1}
         {arriveTangent=0.4235, interpMode=RCIM_Cubic, leaveTangent=0.4235, tangentMode=RCTM_User, time=0.6784, value=0.9948}
         {arriveTangent=0.1347, interpMode=RCIM_Cubic, leaveTangent=0.1347, tangentMode=RCTM_User, time=0.9645, value=1.2125}
         <at 0.9984 1.5>
        ]}}}
     Scale Curve=&RandomRangeFloat(Minimum=3, Maximum=10)
    )
   ) {isSetParameters=true}
   &PendulumSetup(
    Module.Enable Spring Driver=true
    Pendulum Length:NiagaraFloat=$Particles.ZScale
    Pendulum Pivot:NiagaraPosition=$Particles.MeshPosition
    Rigid Pendulum=false
    "Spring Driven Constraint (Tightness)"=5
   )
   &Drag(Drag=-10, Rotational Drag=-10, Use Linear Drag=true, Use Rotational Drag=true)
   &InheritVelocity(
    Inherited Velocity Amount Scale=&Lerp_Vector(A=<0 0 0>, B=<-1 -1 -1>, Alpha=$Particles.NormalizedAge)
    Inherited Velocity Speed Limit=250
    Speed Limit=true
    UseSpeedThreshold=false
   )
   &SolveForcesAndVelocity(Clamp Velocity=false, Limit Acceleration=false)
   &PendulumConstraint()
   &CalculateAccurateVelocity()
   /SetVariables(
    Particles.MeshFacing:Vector3f=&NormalizeVector(
     Vector To Normalize=&Subtract_Vector(
      A=Linked() {linkedVariable {name=Particles.MeshPosition, type=NiagaraPosition}}
      B=Linked() {linkedVariable {name=Particles.Position, type=NiagaraPosition}}
     )
     Safe Normalize=true
     Safe Fallback Threshold=0
     Safe Fallback Vector=<0 0 0.001>
    )
    Particles.RestDirection:Vector3f=&NormalizeVector(Vector To Normalize=$Engine.Owner.Velocity, Safe Normalize=true, Safe Fallback Threshold=0, Safe Fallback Vector=<0 0 0.001>)
   ) {isSetParameters=true}
   &UpdateMeshOrientation(
    Rotation Rate=&Multiply_Float(
     A=&Multiply_Float(A=$Particles.NormalizedAge, B=&RandomRangeFloat(Minimum=0.5, Maximum=1))
     B=&Waveform(
      "Waveform [1]"=::NewEnumerator0
      "[1] Frequency"=&RandomRangeFloat(Minimum=2, Maximum=3)
      "[1] Amplitude Min/Max"=<-0.05 0.05>
      "[1] Scale"=&Multiply_Float(A=&RandomRangeFloat(Minimum=2, Maximum=3), B=$Particles.NormalizedAge)
     )
    )
    Rotation Vector=<0 1 0>
   ) {module=UpdateMeshOrientation001}
   &UpdateMeshOrientation(
    Facing Coordinate Space:ENiagaraCoordinateSpace=::World
    Facing Direction=$Particles.MeshFacing
    Facing Interpolation=true
    Facing Interpolation Rate=15
    Orientation Method=::NewEnumerator1
    Override Incoming Orientation=1
    Override Quat=false
    Reference Axis Vector:Vector3f=<0 0 1>
    Stabilization Mode:ENiagara_OrientVectorCount=::NewEnumerator0
    TransformFacing=true
   )
   &DynamicMaterialParameters(
    Index 2 Param 4=&RandomRangeFloat(Minimum=5, Maximum=10)
    Index 2 Vector 1:NiagaraPosition=$Particles.MeshPosition
    Index 3 Param 4=0.025
    Index 3 Vector 1:NiagaraPosition=$Particles.Position
    Module.Index2Float3Write=true
    Module.Index2VectorWrite=true
    Module.Index3Float3Write=true
    Module.Index3VectorWrite=true
    Write Parameter Index 0=::NewEnumerator0
    Write Parameter Index 2=::NewEnumerator3
    Write Parameter Index 3=::NewEnumerator3
   )
   &ScaleMeshSize(Scale Factor=&MakeVector(
     X=$Particles.XYScale
     Y=$Particles.XYScale
     Z=&VectorLength(Vector=&Subtract_Vector(A=$Particles.MeshPosition, B=Linked() {linkedVariable {name=Particles.Position, type=NiagaraPosition}}))
    ))
   &ScaleColor(Scale Alpha=$Particles.AlphaScale, ScaleA=true, ScaleRGB=true)
   &self:WriteDeath(
    Position:NiagaraPosition=&AddVectorToPosition(Position=$Particles.MeshPosition, Vector=<0 0 -8>)
    Position Array:NiagaraDataInterfaceArrayPosition=$System.DeathArray
   ) {module=WriteDeath}
  ]
  rendererClasses [&NiagaraSpriteRendererProperties, &NiagaraMeshRendererProperties]
  renderers [
   {rendererClass=&NiagaraSpriteRendererProperties}
   {rendererClass=&NiagaraMeshRendererProperties, rendererIndex=1}
   {rendererClass=&NiagaraMeshRendererProperties, rendererIndex=2}
  ]
 }
 JumpDrool {
  emitterUpdateScript [
   &EmitterState(Life Cycle Mode=::NewEnumerator1, Loop Duration=3.5, UseLoopDelay=false)
   &SpawnRate(
    Spawn Probability=&Multiply_Float(A=1, B=&MakeCustomFloatFromBool(Bool=$User.bIsHeadHidden, True Float=0, False Float=1))
    SpawnRate=&MakeCustomFloatFromBool(Bool=$User.SpawnBurst, True Float=&RandomRangeFloat(Minimum=25, Maximum=75), False Float=0)
    Use Spawn Probability=true
   )
  ]
  particleSpawnScript [
   /SetVariables(Particles.PendulumLength:NiagaraFloat=&RandomRangeFloat(Minimum=7, Maximum=25)) {isSetParameters=true}
   &InitializeParticle(
    Color=&FNTimeOfDayColor(Night Color=<rgba 0.6 0.6 0.6 1>, Morning Color=<rgba 1 1 1 1>, Day Color=<rgba 1 1 1 1>, Evening Color=<rgba 1 1 1 1>)
    Lifetime=&UniformRangedFloat(Minimum=0.5, Maximum=1.2)
    Mesh Scale=&Multiply_VectorByFloat(Vector=<0.125 0.125 0.01>, Float=&RandomRangeFloat(Minimum=0.5, Maximum=1))
    Mesh Scale Mode=::NewEnumerator4
    Sprite Size Mode=::NewEnumerator3
    Uniform Sprite Size=2
    UsePositionOffset=false
    Write Material Random=true
   )
   &AddVelocity(
    Module.Use Falloff Distance=false
    Override Seed=false
    Velocity Mode=::NewEnumerator1
    Velocity Speed=&RandomRangeFloat(Minimum=25, Maximum=50)
   )
   &SkeletalMeshLocation(
    Bone Sampling Mode=::NewEnumerator4
    MaskSpawnGroup=false
    Orientation Sampling=::NewEnumerator1
    Sampled Position Offset=<-15 25 0>
    Skeletal Mesh=$System.Skeletal Mesh
   )
   &ShapeLocation(MaskSpawnGroup=false, Override Seed=false, Sphere Radius=3, UseOwnerScale=true)
  ]
  particleUpdateScript [
   &ParticleState()
   &SkeletalMeshLocation(
    Bone Sampling Mode=::NewEnumerator4
    MaskSpawnGroup=false
    Orientation Sampling=::NewEnumerator1
    Position Sampling=::NewEnumerator2
    Sampled Position Offset Coordinate Space:ENiagaraSkelMeshTransforms=::NewEnumerator0
    Skeletal Mesh=$System.Skeletal Mesh
   ) {module=SkeletalMeshLocation001}
   /SetVariables(
    Particles.CompareZ:NiagaraFloat=&OneMinusFloat(Float=&ClampFloat(Float=&Multiply_Float(
       A=&Subtract_Float(
        A=&MakeFloatFromVector(Vector=Linked() {linkedVariable {name=Particles.Position, type=NiagaraPosition}}, Channel=::NewEnumerator2)
        B=&MakeFloatFromVector(Vector=Linked() {linkedVariable {name=Particles.SkeletalMeshLocation.SampledPosition, type=NiagaraPosition}}, Channel=::NewEnumerator2)
       )
       B=0.01
      )))
    Particles.MeshPosition:Vector3f=&Lerp_Vector(
     A=Linked() {linkedVariable {name=Particles.SkeletalMeshLocation.SampledPosition, type=NiagaraPosition}}
     B=Linked() {linkedVariable {name=Particles.Position, type=NiagaraPosition}}
     Alpha=&Add_Float(A=0.9, B=&Multiply_Float(A=0.1, B=$Particles.NormalizedAge))
    )
   ) {isSetParameters=true}
   &GravityForce()
   &SolveForcesAndVelocity(Clamp Velocity=false, Limit Acceleration=false)
   &CalculateAccurateVelocity()
   /SetVariables(Particles.MeshFacing:Vector3f=&NormalizeVector(
     Vector To Normalize=&Subtract_Vector(
      A=Linked() {linkedVariable {name=Particles.SkeletalMeshLocation.SampledPosition, type=NiagaraPosition}}
      B=Linked() {linkedVariable {name=Particles.Position, type=NiagaraPosition}}
     )
     Safe Normalize=true
     Safe Fallback Threshold=0
     Safe Fallback Vector=<0 0 0.001>
    )) {isSetParameters=true}
   &UpdateMeshOrientation(
    Facing Coordinate Space:ENiagaraCoordinateSpace=::World
    Facing Direction=$Particles.MeshFacing
    Facing Interpolation=true
    Facing Interpolation Rate=15
    Orientation Method=::NewEnumerator1
    Override Incoming Orientation=1
    Override Quat=false
    Reference Axis Vector:Vector3f=<0 0 1>
    Stabilization Mode:ENiagara_OrientVectorCount=::NewEnumerator0
    TransformFacing=true
   )
   &DynamicMaterialParameters(
    Index 2 Param 4=0.5
    Index 2 Vector 1:NiagaraPosition=$Particles.MeshPosition
    Index 3 Param 4=0.2
    Index 3 Vector 1:NiagaraPosition=$Particles.Position
    Module.Index2Float3Write=true
    Module.Index2VectorWrite=true
    Module.Index3Float3Write=true
    Module.Index3VectorWrite=true
    Write Parameter Index 0=::NewEnumerator0
    Write Parameter Index 2=::NewEnumerator3
    Write Parameter Index 3=::NewEnumerator3
   )
   /SetVariables(Particles.XYScale:NiagaraFloat=&FloatFromCurve(FloatCurve=DataInterface() {propertyValues {Curve {keys [
         {arriveTangent=-3.383, interpMode=RCIM_Cubic, leaveTangent=-3.383, tangentMode=RCTM_User, time=0, value=1}
         {arriveTangent=-0.5412, interpMode=RCIM_Cubic, leaveTangent=-0.5412, tangentMode=RCTM_User, time=1, value=0}
        ]}}})) {isSetParameters=true}
   &ScaleMeshSize(Scale Factor=&Multiply_VectorByFloat(
     Vector=&MakeVector(
      X=$Particles.XYScale
      Y=$Particles.XYScale
      Z=&ClampFloat(
       Float=&VectorLength(Vector=&Subtract_Vector(A=$Particles.MeshPosition, B=Linked() {linkedVariable {name=Particles.Position, type=NiagaraPosition}}))
       Max=$Particles.PendulumLength
      )
     )
     Float=$Particles.CompareZ
    )) {module=ScaleMeshSize001}
  ]
  rendererClasses [&NiagaraSpriteRendererProperties, &NiagaraMeshRendererProperties]
  renderers [
   {rendererClass=&NiagaraSpriteRendererProperties}
   {rendererClass=&NiagaraMeshRendererProperties, rendererIndex=1}
   {rendererClass=&NiagaraMeshRendererProperties, rendererIndex=2}
  ]
 }
}
summary {
 systemName=NS_Character_Kepler_Idle
 emitters [
  {emitterName=DroolDrips, rendererClasses [&NiagaraSpriteRendererProperties, &NiagaraMeshRendererProperties]}
  {emitterName=DroolMain, rendererClasses [&NiagaraSpriteRendererProperties, &NiagaraMeshRendererProperties]}
  {emitterName=JumpDrool, rendererClasses [&NiagaraSpriteRendererProperties, &NiagaraMeshRendererProperties]}
 ]
 userVariables [
  User.bIsFrontEnd {type=NiagaraBool}
  User.bIsFrontEndPreview {type=NiagaraBool}
  User.bIsHeadHidden {type=NiagaraBool}
  User.ReactiveRate {type=NiagaraFloat, defaultValue=0, description=""}
  User.SpawnBurst {type=NiagaraBool}
 ]
}
