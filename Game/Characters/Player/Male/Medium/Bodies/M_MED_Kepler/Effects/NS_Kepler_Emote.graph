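# Niagara system for the Kepler character's emote that hangs drool strands and drops from the skeletal mesh.
# Two pendulum emitters spawn on age-timed curves, re-sample their bone pivot each frame, stretch meshes from pivot to swinging end, and generate death events.
# Three drop emitters add falling, shrinking droplets; System.RoarCurve blends their downward velocity and drives the spawn rate and fling of DroolDrops002, whose alpha also adjusts for front end.
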NiagaraSystem /Game/Characters/Player/Male/Medium/Bodies/M_MED_Kepler/Effects/NS_Kepler_Emote
dependencies {
 usedDataInterfaces [&NiagaraDataInterfaceCurve, &NiagaraDataInterfaceColorCurve]
 usedDynamicInputs []
 usedModules [
  &SystemState
  &self:SystemSpawnScript.SystemScriptSource.SystemScriptGraph.NiagaraNodeAssignment_0.SetVariables_#1
 ]
}
emitters {
 DroolDrops {
  emitterUpdateScript [&EmitterState()]
  particleSpawnScript [
   /Niagara/Modules/Spawn/Initialization/InitializeParticle(
    Lifetime=&RandomRangeFloat(Minimum=0.25, Maximum=0.5)
    Mesh Scale=<5 0.5 0.5>
    Module.Write Color=true
    Module.Write Lifetime=true
    Module.Write Mass=true
    Module.Write Material Random=true
    Module.Write Position=false
    Module.Write Scale=true
    Module.Write SpriteRotation=false
    Module.Write SpriteSize=true
    Sprite Size=&Vector2DFromFloat(Value=5)
   )
   &AddVelocity(Coordinate Space=::World, Velocity=&Lerp_Vector(A=<0 0 -50>, B=<0 0 0>, Alpha=$System.RoarCurve)) {module=AddVelocity002}
   &AddVelocity(Velocity=&Lerp_Vector(A=<0 0 0>, B=<150 0 50>, Alpha=$System.RoarCurve)) {enabled=false, module=AddVelocity001}
  ]
  particleUpdateScript [
   &ParticleState()
   &DynamicMaterialParameters(
    Index 0 Param 1=&RandomRangeFloat(Minimum=0, Maximum=1)
    Index 0 Param 2=1
    Module.Param0WriteEnabled=true
    Module.Param1WriteEnabled=true
    Module.Param2WriteEnabled=true
    Module.Param3WriteEnabled=true
   )
   &ScaleMeshSize(Scale Factor=&Multiply_VectorByFloat(
     Vector=&MakeVector(
      X=&Lerp_Float(A=0.25, B=1, Alpha=$Particles.NormalizedAge)
      Y=&Lerp_Float(A=1, B=0.5, Alpha=$Particles.NormalizedAge)
      Z=&Lerp_Float(A=1, B=0.5, Alpha=$Particles.NormalizedAge)
     )
     Float=&Lerp_Float(A=1, B=0, Alpha=$Particles.NormalizedAge)
    ))
   &GravityForce()
   &SolveForcesAndVelocity(Clamp Velocity=false, Limit Acceleration=false)
  ]
  rendererClasses [&NiagaraSpriteRendererProperties, &NiagaraMeshRendererProperties]
  renderers [
   {rendererClass=&NiagaraSpriteRendererProperties}
   {rendererClass=&NiagaraMeshRendererProperties, rendererIndex=1}
  ]
 }
 DroolDrops001 {
  emitterUpdateScript [&EmitterState()]
  particleSpawnScript [
   &InitializeParticle(
    Color=<rgba 1 1 1 0.5>
    Lifetime=&RandomRangeFloat(Minimum=0.25, Maximum=0.5)
    Mesh Scale=<5 0.5 0.5>
    Module.Write Color=true
    Module.Write Lifetime=true
    Module.Write Mass=true
    Module.Write Material Random=true
    Module.Write Position=false
    Module.Write Scale=true
    Module.Write SpriteRotation=false
    Module.Write SpriteSize=true
    Sprite Size=&Vector2DFromFloat(Value=5)
   )
   &AddVelocity(Coordinate Space=::World, Velocity=&Lerp_Vector(A=<0 0 -50>, B=<0 0 0>, Alpha=$System.RoarCurve))
   &AddVelocity(Velocity=&Lerp_Vector(A=<0 0 0>, B=<600 0 0>, Alpha=$System.RoarCurve)) {enabled=false, module=AddVelocity001}
  ]
  particleUpdateScript [
   &ParticleState()
   &DynamicMaterialParameters(
    Index 0 Param 1=&RandomRangeFloat(Minimum=0, Maximum=1)
    Index 0 Param 2=1
    Module.Param0WriteEnabled=true
    Module.Param1WriteEnabled=true
    Module.Param2WriteEnabled=true
    Module.Param3WriteEnabled=true
   )
   &ScaleMeshSize(Scale Factor=&Multiply_VectorByFloat(
     Vector=&MakeVector(
      X=&Lerp_Float(A=0.25, B=1, Alpha=$Particles.NormalizedAge)
      Y=&Lerp_Float(A=1, B=0.5, Alpha=$Particles.NormalizedAge)
      Z=&Lerp_Float(A=1, B=0.5, Alpha=$Particles.NormalizedAge)
     )
     Float=&Lerp_Float(A=1, B=0, Alpha=$Particles.NormalizedAge)
    ))
   &GravityForce()
   &SolveForcesAndVelocity(Clamp Velocity=false, Limit Acceleration=false)
  ]
  rendererClasses [&NiagaraSpriteRendererProperties, &NiagaraMeshRendererProperties]
  renderers [
   {rendererClass=&NiagaraSpriteRendererProperties}
   {rendererClass=&NiagaraMeshRendererProperties, rendererIndex=1}
  ]
 }
 DroolDrops002 {
  emitterUpdateScript [
   &EmitterState()
   &SpawnRate(Spawn Probability=0.6, SpawnRate=&Multiply_Float(A=$System.RoarCurve, B=75), Use Spawn Probability=true)
  ]
  particleSpawnScript [
   &InitializeParticle(
    Lifetime=&RandomRangeFloat(Minimum=0.25, Maximum=0.5)
    Mesh Scale=&RandomRangeVector(Minimum=<3 0.3 0.3>, Maximum=<5 0.5 0.5>)
    Module.Write Color=true
    Module.Write Lifetime=true
    Module.Write Mass=true
    Module.Write Material Random=true
    Module.Write Position=false
    Module.Write Scale=true
    Module.Write SpriteRotation=false
    Module.Write SpriteSize=true
    Sprite Size=&Vector2DFromFloat(Value=5)
   )
   &TorusLocation(
    Align Torus With Axis=false
    Handle Radius=0.01
    Large Radius=2
    MaskSpawnGroup=false
    Module.Override Randomness=false
    Torus Origin:NiagaraPosition=<0 0 0>
   )
   /SetVariables(
    Particles.BoneOffset:Vector3f=&RandomRangeVector(Minimum=<0 5 -3>, Maximum=<0 5 0>)
    Particles.TorusPosition:NiagaraPosition=$Particles.Position
   ) {isSetParameters=true}
   /SetVariables(
    Particles.BoneOffset:Vector3f=&RandomRangeVector(Minimum=<0 -2.75 2>, Maximum=<0 -2.75 5>)
    Particles.TorusPosition:NiagaraPosition=$Particles.Position
   ) {enabled=false, isSetParameters=true}
   &SkeletalMeshLocation(
    Bone / Socket Index=2
    Bone Sampling Mode=::NewEnumerator4
    Inherit Velocity=false
    MaskSpawnGroup=false
    Sampled Position Offset=&Add_Vector(A=$Particles.BoneOffset, B=$Particles.TorusPosition)
    Sampled Position Offset Coordinate Space:ENiagaraMeshTransforms=::NewEnumerator3
    Skeletal Mesh=$User.SkeletalMesh
   ) {module=SkeletalMeshLocation002}
   &AddVelocity(Coordinate Space=::World, Velocity=&Lerp_Vector(A=<0 0 -50>, B=<0 0 0>, Alpha=$System.RoarCurve))
   &AddVelocity(Velocity=&Lerp_Vector(A=<0 0 0>, B=&RandomRangeVector(Minimum=<100 0 -750>, Maximum=<750 0 -100>), Alpha=$System.RoarCurve)) {module=AddVelocity001}
  ]
  particleUpdateScript [
   &ParticleState()
   &DynamicMaterialParameters(
    Index 0 Param 1=&RandomRangeFloat(Minimum=0, Maximum=1)
    Index 0 Param 2=1
    Module.Param0WriteEnabled=true
    Module.Param1WriteEnabled=true
    Module.Param2WriteEnabled=true
    Module.Param3WriteEnabled=true
   )
   &ScaleMeshSize(Scale Factor=&Multiply_VectorByFloat(
     Vector=&MakeVector(
      X=&Lerp_Float(A=0.25, B=1, Alpha=$Particles.NormalizedAge)
      Y=&Lerp_Float(A=1, B=0.5, Alpha=$Particles.NormalizedAge)
      Z=&Lerp_Float(A=1, B=0.5, Alpha=$Particles.NormalizedAge)
     )
     Float=&Lerp_Float(A=1, B=0, Alpha=$Particles.NormalizedAge)
    ))
   &ScaleColor(Scale Alpha=&FN_FrontEndFloatModifier(Gameplay=1, Front End=8), ScaleA=true, ScaleRGB=false)
   &GravityForce(Gravity=<0 0 -1500>)
   &SolveForcesAndVelocity(Clamp Velocity=false, Limit Acceleration=false)
  ]
  rendererClasses [&NiagaraSpriteRendererProperties, &NiagaraMeshRendererProperties]
  renderers [
   {rendererClass=&NiagaraSpriteRendererProperties}
   {rendererClass=&NiagaraMeshRendererProperties, rendererIndex=1}
  ]
 }
 DroolPendulum {
  emitterUpdateScript [
   &EmitterState()
   &SpawnRate(
    Spawn Probability=&FloatFromCurve(
     FloatCurve=DataInterface() {propertyValues {Curve {keys [
         {arriveTangent=-0.0634, interpMode=RCIM_Cubic, leaveTangent=-0.0634, tangentMode=RCTM_User, time=4.5, value=0}
         {arriveTangent=0.0041, interpMode=RCIM_Cubic, leaveTangent=0.0041, tangentMode=RCTM_User, time=4.85, value=1}
         {arriveTangent=-0.3387, interpMode=RCIM_Cubic, leaveTangent=-0.3387, tangentMode=RCTM_User, time=5, value=0}
        ]}}}
     CurveIndex=$Engine.System.Age
    )
    SpawnRate=&RandomRangeFloat(Minimum=7, Maximum=20, Evaluation Type=::NewEnumerator1)
    Use Spawn Probability=true
   )
  ]
  particleSpawnScript [
   &TorusLocation(
    Align Torus With Axis=false
    Handle Radius=0.01
    Large Radius=1
    MaskSpawnGroup=false
    Module.Override Randomness=false
    Torus Origin:NiagaraPosition=<0 0 0>
   )
   /SetVariables(Particles.TorusPosition:NiagaraPosition=$Particles.Position) {isSetParameters=true, module=SetVariables_#2}
   &SkeletalMeshLocation(
    Bone Sampling Mode=::NewEnumerator4
    Inherit Velocity=false
    MaskSpawnGroup=false
    Sampled Position Offset=$Particles.TorusPosition
    Sampled Position Offset Coordinate Space:ENiagaraMeshTransforms=::NewEnumerator3
    Skeletal Mesh=$User.SkeletalMesh
   ) {module=SkeletalMeshLocation002}
   &InitializeParticle(
    Lifetime=&UniformRangedFloat(Minimum=0.125, Maximum=0.3)
    Mesh Scale=<0.1 0.1 0.018>
    Module.Write Color=true
    Module.Write Lifetime=true
    Module.Write Mass=true
    Module.Write Material Random=true
    Module.Write Position=false
    Module.Write Scale=true
    Module.Write SpriteRotation=false
    Module.Write SpriteSize=true
    Sprite Size=&Vector2DFromFloat(Value=2)
   )
  ]
  particleUpdateScript [
   &ParticleState()
   /SetVariables(Particles.OrigPos:NiagaraPosition=$Particles.Position) {isSetParameters=true, module=SetVariables_#3}
   &SkeletalMeshLocation(
    Bone Sampling Mode=::NewEnumerator4
    Inherit Velocity=false
    MaskSpawnGroup=false
    Sampled Position Offset=$Particles.TorusPosition
    Sampled Position Offset Coordinate Space:ENiagaraMeshTransforms=::NewEnumerator3
    Skeletal Mesh=$User.SkeletalMesh
   )
   /SetVariables(Particles.MeshPosition:NiagaraPosition=$Particles.Position) {isSetParameters=true, module=SetVariables_#4}
   /SetVariables(Particles.Position:Vector3f=$Particles.OrigPos) {isSetParameters=true, module=SetVariables_#5}
   /SetVariables(
    Particles.AlphaScale:NiagaraFloat=&FloatFromCurve(FloatCurve=DataInterface() {propertyValues {Curve {keys [
         {arriveTangent=-0.0275, interpMode=RCIM_Cubic, leaveTangent=-0.0275, tangentMode=RCTM_User, time=0.95, value=1}
         {arriveTangent=-34.2468, interpMode=RCIM_Cubic, leaveTangent=-34.2466, tangentMode=RCTM_User, time=1, value=0}
        ]}}})
    Particles.XYScale:NiagaraFloat=&FloatFromCurve(
     FloatCurve=DataInterface() {propertyValues {Curve {keys [
         {arriveTangent=6.0986, interpMode=RCIM_Cubic, leaveTangent=6.0986, tangentMode=RCTM_User, time=0, value=0}
         {arriveTangent=-1.5439, interpMode=RCIM_Cubic, leaveTangent=-1.5439, tangentMode=RCTM_User, time=0.25, value=1}
         {arriveTangent=-0.3409, interpMode=RCIM_Cubic, leaveTangent=-0.3409, tangentMode=RCTM_User, time=0.6, value=0.6082}
         {arriveTangent=-2.6194, interpMode=RCIM_Cubic, leaveTangent=-2.6194, tangentMode=RCTM_User, time=0.9823, value=0.214}
         <at 1 0.05>
        ]}}}
     Scale Curve=&RandomRangeFloat(Minimum=0.4, Maximum=1)
    )
    Particles.ZScale:NiagaraFloat=&FloatFromCurve(
     FloatCurve=DataInterface() {propertyValues {Curve {keys [
         {arriveTangent=3.0916, interpMode=RCIM_Cubic, leaveTangent=3.0916, tangentMode=RCTM_User, time=0, value=0.1}
         {arriveTangent=0.4235, interpMode=RCIM_Cubic, leaveTangent=0.4235, tangentMode=RCTM_User, time=0.6784, value=0.9948}
         {arriveTangent=0.1347, interpMode=RCIM_Cubic, leaveTangent=0.1347, tangentMode=RCTM_User, time=0.9645, value=1.2125}
         <at 0.9984 1.5>
        ]}}}
     Scale Curve=&RandomRangeFloat(Minimum=3, Maximum=10)
    )
   ) {isSetParameters=true, module=SetVariables_#6}
   &PendulumSetup(
    Module.Enable Spring Driver=false
    Pendulum Length:NiagaraFloat=$Particles.ZScale
    Pendulum Pivot:Vector3f=$Particles.MeshPosition
    Rigid Pendulum=false
   )
   &Drag(Drag=-10, Rotational Drag=-10, Use Linear Drag=true, Use Rotational Drag=true)
   &AddVelocity(Coordinate Space=::World, Velocity=<0 0 -40>)
   &AddVelocity(Velocity=&Lerp_Vector(A=<0 0 0>, B=<100 0 50>, Alpha=$System.RoarCurve)) {enabled=false, module=AddVelocity001}
   &InheritVelocity(Module.Limit Speed=true, Velocity Scale=&Lerp_Vector(A=<-1 -1 -1>, B=<0 0 0>, Alpha=$Particles.NormalizedAge))
   &SolveForcesAndVelocity(Clamp Velocity=false, Limit Acceleration=false)
   &PendulumConstraint()
   &CalculateAccurateVelocity()
   /SetVariables(
    Particles.MeshFacing:Vector3f=&NormalizeVector(Vector To Normalize=&Subtract_Vector(A=$Particles.MeshPosition, B=Linked() {linkedVariable {name=Particles.Position, type=NiagaraPosition}}))
    Particles.RestDirection:Vector3f=/Niagara/DynamicInputs/Helpers/SafeNormalizeVector(Vector To Normalize:Vector3f=$Engine.Owner.Velocity, Fallback Vector:Vector3f=<0 0 0>)
   ) {isSetParameters=true, module=SetVariables_#7}
   /Niagara/Modules/Update/Orientation/V2/OrientMeshToVector(
    Look At Direction:Vector3f=&Multiply_VectorByFloat(Vector=$Particles.MeshFacing, Float=1)
    Override Quat=false
    Reference Axis Vector:Vector3f=<0 0 1>
    UseOptionalSideAxis=false
    UseOptionalUpAxis=false
   )
   &DynamicMaterialParameters(
    Index 2 Param 1=HlslExpression() {hlslExpression=Particles.MeshPosition.x}
    Index 2 Param 2=HlslExpression() {hlslExpression=Particles.MeshPosition.y}
    Index 2 Param 3=HlslExpression() {hlslExpression=Particles.MeshPosition.z}
    Index 2 Param 4=&RandomRangeFloat(Minimum=2, Maximum=5)
    Index 3 Param 1=&MakeFloatFromVector(Vector=Linked() {linkedVariable {name=Particles.Position, type=NiagaraPosition}}, Channel=::NewEnumerator0)
    Index 3 Param 2=&MakeFloatFromVector(Vector=Linked() {linkedVariable {name=Particles.Position, type=NiagaraPosition}}, Channel=::NewEnumerator1)
    Index 3 Param 3=&MakeFloatFromVector(Vector=Linked() {linkedVariable {name=Particles.Position, type=NiagaraPosition}}, Channel=::NewEnumerator2)
    Index 3 Param 4=0.025
    Module.Param0WriteEnabled2=true
    Module.Param0WriteEnabled3=true
    Module.Param1WriteEnabled2=true
    Module.Param1WriteEnabled3=true
    Module.Param2WriteEnabled2=true
    Module.Param2WriteEnabled3=true
    Module.Param3WriteEnabled2=true
    Module.Param3WriteEnabled3=true
    Write Parameter Index 0=false
    Write Parameter Index 2=true
    Write Parameter Index 3=true
   )
   &ScaleMeshSize(Scale Factor=&MakeVector(
     X=$Particles.XYScale
     Y=$Particles.XYScale
     Z=&VectorLength(Vector=&Subtract_Vector(A=$Particles.MeshPosition, B=Linked() {linkedVariable {name=Particles.Position, type=NiagaraPosition}}))
    ))
   &Color(Color=&ColorFromCurve(
     ColorCurve=DataInterface() {propertyValues {
       AlphaCurve {keys [<at 0 20 RCIM_Cubic>]}
       BlueCurve {keys [<at 0 0 RCIM_Cubic>,<at 0.5257 0>,<at 1 5 RCIM_Cubic>]}
       GreenCurve {keys [<at 0 0 RCIM_Cubic>,<at 0.5257 5>,<at 1 0 RCIM_Cubic>]}
       RedCurve {keys [<at 0 5 RCIM_Cubic>,<at 0.5257 0>,<at 1 0 RCIM_Cubic>]}
      }}
     CurveIndex=&Divide_Float(A=&MakeFloatFromInt(Int=$Particles.PickBone), B=366)
    ))
   &ScaleColor(Scale Alpha=$Particles.AlphaScale, ScaleA=true, ScaleRGB=false)
   &GenerateDeathEvent()
  ]
  rendererClasses [&NiagaraSpriteRendererProperties, &NiagaraMeshRendererProperties]
  renderers [
   {rendererClass=&NiagaraSpriteRendererProperties}
   {rendererClass=&NiagaraMeshRendererProperties, rendererIndex=1}
   {rendererClass=&NiagaraSpriteRendererProperties, rendererIndex=2}
  ]
 }
 DroolPendulum001 {
  emitterUpdateScript [
   &EmitterState()
   &SpawnRate(
    Spawn Probability=&FloatFromCurve(
     FloatCurve=DataInterface() {propertyValues {Curve {keys [
         {arriveTangent=0.2088, interpMode=RCIM_Cubic, leaveTangent=0.2088, tangentMode=RCTM_User, time=0, value=0}
         {arriveTangent=0.2374, interpMode=RCIM_Cubic, leaveTangent=0.2374, tangentMode=RCTM_User, time=3.1715, value=0.2598}
         {arriveTangent=0.0041, interpMode=RCIM_Cubic, leaveTangent=0.0041, tangentMode=RCTM_User, time=4.25, value=1}
         {arriveTangent=-0.1194, interpMode=RCIM_Cubic, leaveTangent=-0.1194, tangentMode=RCTM_User, time=6.1878, value=0.2782}
        ]}}}
     CurveIndex=$Engine.System.Age
    )
    SpawnRate=&RandomRangeFloat(Minimum=10, Maximum=30, Evaluation Type=::NewEnumerator1)
    Use Spawn Probability=true
   )
   &SpawnBurst_Instantaneous(Spawn Count=2, Use Spawn Probability=false)
  ]
  particleSpawnScript [
   &TorusLocation(
    Align Torus With Axis=false
    Handle Radius=0.01
    Large Radius=3
    MaskSpawnGroup=false
    Module.Override Randomness=false
    Torus Origin:NiagaraPosition=<0 0 0>
   )
   /SetVariables(
    Particles.BoneOffset:Vector3f=&RandomRangeVector(Minimum=<0 6.75 -4>, Maximum=<0 6.75 1>)
    Particles.TorusPosition:NiagaraPosition=$Particles.Position
   ) {isSetParameters=true, module=SetVariables_#2}
   /SetVariables(
    Particles.BoneOffset:Vector3f=&RandomRangeVector(Minimum=<0 -5 0>, Maximum=<0 -5 7>)
    Particles.TorusPosition:NiagaraPosition=$Particles.Position
   ) {enabled=false, isSetParameters=true}
   &SkeletalMeshLocation(
    Bone / Socket Index=2
    Bone Sampling Mode=::NewEnumerator4
    Inherit Velocity=false
    MaskSpawnGroup=false
    Sampled Position Offset=&Add_Vector(A=$Particles.BoneOffset, B=$Particles.TorusPosition)
    Sampled Position Offset Coordinate Space:ENiagaraMeshTransforms=::NewEnumerator3
    Skeletal Mesh=$User.SkeletalMesh
   ) {module=SkeletalMeshLocation002}
   &KillParticlesInVolume(
    Box Size:Vector3f=<5 6 10>
    Kill Shape=::NewEnumerator1
    Origin Offset:Vector3f=<0 -3 2.5>
    Volume Origin:Vector3f=$Output.SkeletalMeshLocation002.SampledPosition
   )
   &InitializeParticle(
    Lifetime=&UniformRangedFloat(Minimum=0.25, Maximum=0.5)
    Mesh Scale=<0.1 0.1 0.018>
    Module.Write Color=true
    Module.Write Lifetime=true
    Module.Write Mass=true
    Module.Write Material Random=true
    Module.Write Position=false
    Module.Write Scale=true
    Module.Write SpriteRotation=false
    Module.Write SpriteSize=true
    Sprite Size=&Vector2DFromFloat(Value=2)
   )
  ]
  particleUpdateScript [
   &ParticleState()
   /SetVariables(Particles.OrigPos:NiagaraPosition=$Particles.Position) {isSetParameters=true, module=SetVariables_#3}
   &SkeletalMeshLocation(
    Bone / Socket Index=2
    Bone Sampling Mode=::NewEnumerator4
    Inherit Velocity=false
    MaskSpawnGroup=false
    Sampled Position Offset=&Add_Vector(A=$Particles.TorusPosition, B=$Particles.BoneOffset)
    Sampled Position Offset Coordinate Space:ENiagaraMeshTransforms=::NewEnumerator3
    Skeletal Mesh=$User.SkeletalMesh
   )
   /SetVariables(Particles.MeshPosition:NiagaraPosition=$Particles.Position) {isSetParameters=true, module=SetVariables_#4}
   /SetVariables(Particles.Position:Vector3f=$Particles.OrigPos) {isSetParameters=true, module=SetVariables_#5}
   /SetVariables(
    Particles.AlphaScale:NiagaraFloat=&FloatFromCurve(FloatCurve=DataInterface() {propertyValues {Curve {keys [
         {arriveTangent=-0.0275, interpMode=RCIM_Cubic, leaveTangent=-0.0275, tangentMode=RCTM_User, time=0.95, value=1}
         {arriveTangent=-34.2468, interpMode=RCIM_Cubic, leaveTangent=-34.2466, tangentMode=RCTM_User, time=1, value=0}
        ]}}})
    Particles.XYScale:NiagaraFloat=&FloatFromCurve(
     FloatCurve=DataInterface() {propertyValues {Curve {keys [
         {arriveTangent=10.8032, interpMode=RCIM_Cubic, leaveTangent=10.8032, tangentMode=RCTM_User, time=0, value=0}
         {arriveTangent=0.044, interpMode=RCIM_Cubic, leaveTangent=0.044, tangentMode=RCTM_User, time=0.25, value=1}
         {arriveTangent=-0.3409, interpMode=RCIM_Cubic, leaveTangent=-0.3409, tangentMode=RCTM_User, time=0.6, value=0.6082}
         {arriveTangent=-2.6194, interpMode=RCIM_Cubic, leaveTangent=-2.6194, tangentMode=RCTM_User, time=0.9823, value=0.214}
         <at 1 0.05>
        ]}}}
     Scale Curve=&RandomRangeFloat(Minimum=0.4, Maximum=1)
    )
    Particles.ZScale:NiagaraFloat=&FloatFromCurve(
     FloatCurve=DataInterface() {propertyValues {Curve {keys [
         {arriveTangent=2.4615, interpMode=RCIM_Cubic, leaveTangent=2.4615, tangentMode=RCTM_User, time=0, value=0.01}
         {arriveTangent=0.8133, interpMode=RCIM_Cubic, leaveTangent=0.8133, tangentMode=RCTM_User, time=0.6784, value=0.9948}
         {arriveTangent=1.2917, interpMode=RCIM_Cubic, leaveTangent=1.2917, tangentMode=RCTM_User, time=0.9645, value=1.2125}
         <at 0.9984 1.5>
        ]}}}
     Scale Curve=&RandomRangeFloat(Minimum=3, Maximum=10)
    )
   ) {isSetParameters=true, module=SetVariables_#6}
   &PendulumSetup(
    Module.Enable Spring Driver=false
    Pendulum Length:NiagaraFloat=$Particles.ZScale
    Pendulum Pivot:Vector3f=$Particles.MeshPosition
    Rigid Pendulum=false
   )
   &Drag(Drag=-10, Rotational Drag=-10, Use Linear Drag=true, Use Rotational Drag=true)
   &AddVelocity(Coordinate Space=::World, Velocity=<0 0 -40>) {module=AddVelocity001}
   &AddVelocity(Velocity=&Lerp_Vector(
     A=<0 0 0>
     B=&RandomRangeVector(Minimum=<75 0 -150>, Maximum=<150 0 -75>, Evaluation Type=::NewEnumerator1)
     Alpha=&Multiply_Float(
      A=$System.RoarCurve
      B=&MakeCustomFloatFromBool(
       Bool=&SetBoolByFloatComparison(A=&MakeFloatFromVector(Vector=$Particles.Initial.BoneOffset, Channel=::NewEnumerator2), B=5)
       True Float=1
       False Float=0
      )
     )
    )) {enabled=false}
   &AddVelocity(Velocity=&Lerp_Vector(
     A=<0 0 0>
     B=<2000 0 0>
     Alpha=&Multiply_Float(
      A=$System.RoarCurve
      B=&Power(
       a=&Multiply_Float(A=&MakeFloatFromVector(Vector=$Particles.Initial.BoneOffset, Channel=::NewEnumerator2), B=0.1429)
       b=1.1
      )
     )
    )) {enabled=false, module=AddVelocity003}
   &AddVelocity(
    Scale Added Velocity=&VectorFromFloat(Value=&Add_Float(
      A=1
      B=&Multiply_Float(A=&MakeFloatFromVector(Vector=$Particles.Initial.BoneOffset, Channel=::NewEnumerator2), B=-0.1429)
     ))
    Velocity=&Lerp_Vector(A=<0 0 0>, B=<200 0 0>, Alpha=$System.RoarCurve)
   ) {enabled=false, module=AddVelocity002}
   &InheritVelocity(Module.Limit Speed=true, Velocity Scale=&Lerp_Vector(A=<-1 -1 -1>, B=<0 0 0>, Alpha=$Particles.NormalizedAge))
   &SolveForcesAndVelocity(Clamp Velocity=false, Limit Acceleration=false)
   &PendulumConstraint()
   &CalculateAccurateVelocity()
   /SetVariables(
    Particles.MeshFacing:Vector3f=&NormalizeVector(Vector To Normalize=&Subtract_Vector(A=$Particles.MeshPosition, B=Linked() {linkedVariable {name=Particles.Position, type=NiagaraPosition}}))
    Particles.RestDirection:Vector3f=&SafeNormalizeVector(Vector To Normalize:Vector3f=$Engine.Owner.Velocity, Fallback Vector:Vector3f=<0 0 0>)
   ) {isSetParameters=true, module=SetVariables_#7}
   &OrientMeshToVector(
    Look At Direction:Vector3f=&Multiply_VectorByFloat(Vector=$Particles.MeshFacing, Float=1)
    Override Quat=false
    Reference Axis Vector:Vector3f=<0 0 1>
    UseOptionalSideAxis=false
    UseOptionalUpAxis=false
   )
   &DynamicMaterialParameters(
    Index 2 Param 1=HlslExpression() {hlslExpression=Particles.MeshPosition.x}
    Index 2 Param 2=HlslExpression() {hlslExpression=Particles.MeshPosition.y}
    Index 2 Param 3=HlslExpression() {hlslExpression=Particles.MeshPosition.z}
    Index 2 Param 4=&RandomRangeFloat(Minimum=2, Maximum=5)
    Index 3 Param 1=&MakeFloatFromVector(Vector=Linked() {linkedVariable {name=Particles.Position, type=NiagaraPosition}}, Channel=::NewEnumerator0)
    Index 3 Param 2=&MakeFloatFromVector(Vector=Linked() {linkedVariable {name=Particles.Position, type=NiagaraPosition}}, Channel=::NewEnumerator1)
    Index 3 Param 3=&MakeFloatFromVector(Vector=Linked() {linkedVariable {name=Particles.Position, type=NiagaraPosition}}, Channel=::NewEnumerator2)
    Index 3 Param 4=0.025
    Module.Param0WriteEnabled2=true
    Module.Param0WriteEnabled3=true
    Module.Param1WriteEnabled2=true
    Module.Param1WriteEnabled3=true
    Module.Param2WriteEnabled2=true
    Module.Param2WriteEnabled3=true
    Module.Param3WriteEnabled2=true
    Module.Param3WriteEnabled3=true
    Write Parameter Index 0=false
    Write Parameter Index 2=true
    Write Parameter Index 3=true
   )
   &ScaleMeshSize(Scale Factor=&MakeVector(
     X=$Particles.XYScale
     Y=$Particles.XYScale
     Z=&VectorLength(Vector=&Subtract_Vector(A=$Particles.MeshPosition, B=Linked() {linkedVariable {name=Particles.Position, type=NiagaraPosition}}))
    ))
   &Color(Color=&ColorFromCurve(
     ColorCurve=DataInterface() {propertyValues {
       AlphaCurve {keys [<at 0 20 RCIM_Cubic>]}
       BlueCurve {keys [<at 0 0 RCIM_Cubic>,<at 0.5257 0>,<at 1 5 RCIM_Cubic>]}
       GreenCurve {keys [<at 0 0 RCIM_Cubic>,<at 0.5257 5>,<at 1 0 RCIM_Cubic>]}
       RedCurve {keys [<at 0 5 RCIM_Cubic>,<at 0.5257 0>,<at 1 0 RCIM_Cubic>]}
      }}
     CurveIndex=&Multiply_Float(A=&MakeFloatFromVector(Vector=$Particles.Initial.BoneOffset, Channel=::NewEnumerator2), B=0.1429)
    )) {enabled=false}
   &ScaleColor(Scale Alpha=$Particles.AlphaScale, ScaleA=true, ScaleRGB=false)
   &GenerateDeathEvent(Event Condition:NiagaraBool=&SetBoolByFloatComparison(A=$Particles.NormalizedAge, B=0.1))
  ]
  rendererClasses [&NiagaraSpriteRendererProperties, &NiagaraMeshRendererProperties]
  renderers [
   {rendererClass=&NiagaraSpriteRendererProperties}
   {rendererClass=&NiagaraMeshRendererProperties, rendererIndex=1}
  ]
 }
}
summary {
 systemName=NS_Kepler_Emote
 emitters [
  {emitterName=DroolDrops, rendererClasses [&NiagaraSpriteRendererProperties, &NiagaraMeshRendererProperties]}
  {emitterName=DroolPendulum, rendererClasses [&NiagaraSpriteRendererProperties, &NiagaraMeshRendererProperties]}
  {emitterName=DroolDrops001, rendererClasses [&NiagaraSpriteRendererProperties, &NiagaraMeshRendererProperties]}
  {
   emitterName=DroolPendulum001
   rendererClasses [&NiagaraSpriteRendererProperties, &NiagaraMeshRendererProperties]
  }
  {emitterName=DroolDrops002, rendererClasses [&NiagaraSpriteRendererProperties, &NiagaraMeshRendererProperties]}
 ]
 userVariables [
  User.bIsFrontEnd {type=NiagaraBool}
  User.bIsFrontEndPreview {type=NiagaraBool}
  User.SkeletalMesh {
   type=NiagaraDataInterfaceSkeletalMesh
   defaultValue {
    struct=NiagaraExt_VariableValue_DataInterface
    value {dataInterface=&self:NiagaraDataInterfaceSkeletalMesh_0}
   }
  }
 ]
}
