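# Niagara system for the Kepler player body that sheds drool-like strands and drops from the character's skeletal mesh.
# DroolPendulum hangs swinging, stretched meshes from sampled bone locations using pendulum constraints, spawning more as ReactiveRate rises.
# JumpDrool flings gravity-driven strands driven by JumpTime and SpawnBurst, while DroolDrops defines short-lived, shrinking falling droplets.
# Both strand emitters orient meshes toward their mesh anchor and pass endpoint positions to materials through dynamic parameters.
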NiagaraSystem /Game/Characters/Player/Male/Medium/Bodies/M_MED_Kepler/Effects/NS_Kepler_Skin
dependencies {
 usedDataInterfaces [&NiagaraDataInterfaceCurve, &NiagaraDataInterfaceColorCurve]
 usedDynamicInputs []
 usedModules [&SystemState]
}
emitters {
 DroolDrops {
  emitterUpdateScript [&EmitterState()]
  particleSpawnScript [
   /Niagara/Modules/Spawn/Initialization/InitializeParticle(
    Lifetime=&RandomRangeFloat(Minimum=0.25, Maximum=0.5)
    Mesh Scale=<5 0.5 0.5>
    Module.Write Color=true
    Module.Write Lifetime=true
    Module.Write Mass=true
    Module.Write Material Random=true
    Module.Write Position=false
    Module.Write Scale=true
    Module.Write SpriteRotation=false
    Module.Write SpriteSize=true
    Sprite Size=&Vector2DFromFloat(Value=5)
   )
   &AddVelocity(Velocity=<0 0 -50>)
  ]
  particleUpdateScript [
   &ParticleState()
   &DynamicMaterialParameters(
    Index 0 Param 1=&RandomRangeFloat(Minimum=0, Maximum=1)
    Index 0 Param 2=1
    Module.Param0WriteEnabled=true
    Module.Param1WriteEnabled=true
    Module.Param2WriteEnabled=true
    Module.Param3WriteEnabled=true
   )
   &ScaleMeshSize(Scale Factor=&Multiply_VectorByFloat(
     Vector=&MakeVector(
      X=&Lerp_Float(A=0.25, B=1, Alpha=$Particles.NormalizedAge)
      Y=&Lerp_Float(A=1, B=0.5, Alpha=$Particles.NormalizedAge)
      Z=&Lerp_Float(A=1, B=0.5, Alpha=$Particles.NormalizedAge)
     )
     Float=&Lerp_Float(A=1, B=0, Alpha=$Particles.NormalizedAge)
    ))
   &GravityForce()
   &SolveForcesAndVelocity(Clamp Velocity=false, Limit Acceleration=false)
  ]
  rendererClasses [&NiagaraSpriteRendererProperties, &NiagaraMeshRendererProperties]
  renderers [
   {rendererClass=&NiagaraSpriteRendererProperties}
   {rendererClass=&NiagaraMeshRendererProperties, rendererIndex=1}
  ]
 }
 DroolPendulum {
  emitterUpdateScript [
   &EmitterState()
   &SpawnRate(
    Spawn Probability=0.1
    SpawnRate=&RandomRangeFloat(Minimum=1, Maximum=6, Evaluation Type=::NewEnumerator1)
    Use Spawn Probability=true
   )
   &SpawnRate(Spawn Probability=0.25, SpawnRate=&Lerp_Float(A=0, B=100, Alpha=$User.ReactiveRate), Use Spawn Probability=true) {module=SpawnRate001}
  ]
  particleSpawnScript [
   &TorusLocation(
    Align Torus With Axis=false
    Handle Radius=0.01
    Large Radius=3
    MaskSpawnGroup=false
    Module.Override Randomness=false
    Torus Origin:NiagaraPosition=<0 0 0>
   )
   /SetVariables(
    Particles.BoneOffset:Vector3f=&RandomRangeVector(Minimum=<0 6.75 -4>, Maximum=<0 6.75 1>)
    Particles.TorusPosition:NiagaraPosition=$Particles.Position
   ) {isSetParameters=true}
   &SkeletalMeshLocation(
    Bone Sampling Mode=::NewEnumerator4
    Inherit Velocity=false
    MaskSpawnGroup=false
    Sampled Position Offset=&Add_Vector(A=$Particles.BoneOffset, B=$Particles.TorusPosition)
    Sampled Position Offset Coordinate Space:ENiagaraMeshTransforms=::NewEnumerator3
    Skeletal Mesh=$User.SkeletalMesh
   ) {module=SkeletalMeshLocation002}
   &KillParticlesInVolume(
    Box Size:Vector3f=<5 6 10>
    Kill Shape=::NewEnumerator1
    Origin Offset:Vector3f=<0 3 3>
    Volume Origin:Vector3f=$Output.SkeletalMeshLocation002.SampledPosition
   )
   &InitializeParticle(
    Lifetime=&Lerp_Float(
     A=&UniformRangedFloat(Minimum=0.5, Maximum=1)
     B=&UniformRangedFloat(Minimum=0.25, Maximum=0.5)
     Alpha=$User.ReactiveRate
    )
    Mesh Scale=<0.1 0.1 0.018>
    Module.Write Color=true
    Module.Write Lifetime=true
    Module.Write Mass=true
    Module.Write Material Random=true
    Module.Write Position=false
    Module.Write Scale=true
    Module.Write SpriteRotation=false
    Module.Write SpriteSize=true
    Sprite Size=&Vector2DFromFloat(Value=2)
   )
  ]
  particleUpdateScript [
   &ParticleState()
   /SetVariables(Particles.OrigPos:NiagaraPosition=$Particles.Position) {isSetParameters=true}
   &SkeletalMeshLocation(
    Bone Sampling Mode=::NewEnumerator4
    Inherit Velocity=false
    MaskSpawnGroup=false
    Sampled Position Offset=&Add_Vector(A=$Particles.BoneOffset, B=$Particles.TorusPosition)
    Sampled Position Offset Coordinate Space:ENiagaraMeshTransforms=::NewEnumerator3
    Skeletal Mesh=$User.SkeletalMesh
   )
   /SetVariables(Particles.MeshPosition:NiagaraPosition=$Particles.Position) {isSetParameters=true}
   /SetVariables(Particles.Position:Vector3f=$Particles.OrigPos) {isSetParameters=true}
   /SetVariables(
    Particles.AlphaScale:NiagaraFloat=&FloatFromCurve(FloatCurve=DataInterface() {propertyValues {Curve {keys [
         {arriveTangent=-0.0275, interpMode=RCIM_Cubic, leaveTangent=-0.0275, tangentMode=RCTM_User, time=0.95, value=1}
         {arriveTangent=-34.2468, interpMode=RCIM_Cubic, leaveTangent=-34.2466, tangentMode=RCTM_User, time=1, value=0}
        ]}}})
    Particles.XYScale:NiagaraFloat=&FloatFromCurve(
     FloatCurve=DataInterface() {propertyValues {Curve {keys [
         {arriveTangent=6.0986, interpMode=RCIM_Cubic, leaveTangent=6.0986, tangentMode=RCTM_User, time=0, value=0}
         {arriveTangent=-1.5439, interpMode=RCIM_Cubic, leaveTangent=-1.5439, tangentMode=RCTM_User, time=0.25, value=1}
         {arriveTangent=-0.3409, interpMode=RCIM_Cubic, leaveTangent=-0.3409, tangentMode=RCTM_User, time=0.6, value=0.6082}
         {arriveTangent=-2.6194, interpMode=RCIM_Cubic, leaveTangent=-2.6194, tangentMode=RCTM_User, time=0.9823, value=0.214}
         <at 1 0.05>
        ]}}}
     Scale Curve=&RandomRangeFloat(Minimum=0.4, Maximum=1)
    )
    Particles.ZScale:NiagaraFloat=&FloatFromCurve(
     FloatCurve=DataInterface() {propertyValues {Curve {keys [
         {arriveTangent=3.0916, interpMode=RCIM_Cubic, leaveTangent=3.0916, tangentMode=RCTM_User, time=0, value=0.1}
         {arriveTangent=0.4235, interpMode=RCIM_Cubic, leaveTangent=0.4235, tangentMode=RCTM_User, time=0.6784, value=0.9948}
         {arriveTangent=0.1347, interpMode=RCIM_Cubic, leaveTangent=0.1347, tangentMode=RCTM_User, time=0.9645, value=1.2125}
         <at 0.9984 1.5>
        ]}}}
     Scale Curve=&RandomRangeFloat(Minimum=3, Maximum=10)
    )
   ) {isSetParameters=true}
   &PendulumSetup(
    Module.Enable Spring Driver=false
    Pendulum Length:NiagaraFloat=$Particles.ZScale
    Pendulum Pivot:Vector3f=$Particles.MeshPosition
    Rigid Pendulum=false
   )
   &Drag(Drag=-10, Rotational Drag=-10, Use Linear Drag=true, Use Rotational Drag=true)
   &AddVelocity(Velocity=<0 0 -40>)
   &InheritVelocity(Module.Limit Speed=true, Velocity Scale=&Lerp_Vector(A=<-1 -1 -1>, B=<0 0 0>, Alpha=$Particles.NormalizedAge))
   &SolveForcesAndVelocity(Clamp Velocity=false, Limit Acceleration=false)
   &PendulumConstraint()
   &CalculateAccurateVelocity()
   /SetVariables(
    Particles.MeshFacing:Vector3f=&NormalizeVector(Vector To Normalize=&Subtract_Vector(A=$Particles.MeshPosition, B=Linked() {linkedVariable {name=Particles.Position, type=NiagaraPosition}}))
    Particles.RestDirection:Vector3f=/Niagara/DynamicInputs/Helpers/SafeNormalizeVector(Vector To Normalize:Vector3f=$Engine.Owner.Velocity, Fallback Vector:Vector3f=<0 0 0>)
   ) {isSetParameters=true, module=SetVariables_#1}
   /Niagara/Modules/Update/Orientation/V2/OrientMeshToVector(
    Look At Direction:Vector3f=&Multiply_VectorByFloat(Vector=$Particles.MeshFacing, Float=1)
    Override Quat=false
    Reference Axis Vector:Vector3f=<0 0 1>
    UseOptionalSideAxis=false
    UseOptionalUpAxis=false
   )
   &DynamicMaterialParameters(
    Index 2 Param 1=HlslExpression() {hlslExpression=Particles.MeshPosition.x}
    Index 2 Param 2=HlslExpression() {hlslExpression=Particles.MeshPosition.y}
    Index 2 Param 3=HlslExpression() {hlslExpression=Particles.MeshPosition.z}
    Index 2 Param 4=&RandomRangeFloat(Minimum=2, Maximum=5)
    Index 3 Param 1=&MakeFloatFromVector(Vector=Linked() {linkedVariable {name=Particles.Position, type=NiagaraPosition}}, Channel=::NewEnumerator0)
    Index 3 Param 2=&MakeFloatFromVector(Vector=Linked() {linkedVariable {name=Particles.Position, type=NiagaraPosition}}, Channel=::NewEnumerator1)
    Index 3 Param 3=&MakeFloatFromVector(Vector=Linked() {linkedVariable {name=Particles.Position, type=NiagaraPosition}}, Channel=::NewEnumerator2)
    Index 3 Param 4=0.025
    Module.Param0WriteEnabled2=true
    Module.Param0WriteEnabled3=true
    Module.Param1WriteEnabled2=true
    Module.Param1WriteEnabled3=true
    Module.Param2WriteEnabled2=true
    Module.Param2WriteEnabled3=true
    Module.Param3WriteEnabled2=true
    Module.Param3WriteEnabled3=true
    Write Parameter Index 0=false
    Write Parameter Index 2=true
    Write Parameter Index 3=true
   )
   &ScaleMeshSize(Scale Factor=&MakeVector(
     X=$Particles.XYScale
     Y=$Particles.XYScale
     Z=&VectorLength(Vector=&Subtract_Vector(A=$Particles.MeshPosition, B=Linked() {linkedVariable {name=Particles.Position, type=NiagaraPosition}}))
    ))
   &Color(Color=&ColorFromCurve(
     ColorCurve=DataInterface() {propertyValues {
       AlphaCurve {keys [<at 0 20 RCIM_Cubic>]}
       BlueCurve {keys [<at 0 0 RCIM_Cubic>,<at 0.5257 0>,<at 1 5 RCIM_Cubic>]}
       GreenCurve {keys [<at 0 0 RCIM_Cubic>,<at 0.5257 5>,<at 1 0 RCIM_Cubic>]}
       RedCurve {keys [<at 0 5 RCIM_Cubic>,<at 0.5257 0>,<at 1 0 RCIM_Cubic>]}
      }}
     CurveIndex=&Divide_Float(A=&MakeFloatFromInt(Int=$Particles.PickBone), B=366)
    )) {enabled=false}
   &ScaleColor(Scale Alpha=$Particles.AlphaScale, ScaleA=true, ScaleRGB=false)
   &GenerateDeathEvent(Event Condition:NiagaraBool=&SetBoolByFloatComparison(A=$Particles.NormalizedAge, B=0.1))
  ]
  rendererClasses [&NiagaraSpriteRendererProperties, &NiagaraMeshRendererProperties]
  renderers [
   {rendererClass=&NiagaraSpriteRendererProperties}
   {rendererClass=&NiagaraMeshRendererProperties, rendererIndex=1}
  ]
 }
 JumpDrool {
  emitterUpdateScript [
   &EmitterState(Life Cycle Mode=::NewEnumerator1, Loop Duration=3.5, UseLoopDelay=false)
   &SpawnRate(
    Spawn Probability=&ClampFloat(Float=$User.JumpTime)
    SpawnRate=&Lerp_Float(A=4, B=25, Alpha=&ClampFloat(Float=$User.JumpTime))
    Use Spawn Probability=true
   ) {module=SpawnRate002}
   &SpawnRate(
    SpawnRate=&MakeCustomFloatFromBool(Bool=$User.SpawnBurst, True Float=&RandomRangeFloat(Minimum=25, Maximum=75), False Float=0)
    Use Spawn Probability=false
   ) {module=SpawnRate001}
  ]
  particleSpawnScript [
   /SetVariables(Particles.PendulumLength:NiagaraFloat=&RandomRangeFloat(Minimum=7, Maximum=25)) {isSetParameters=true}
   &InitializeParticle(
    Lifetime=&UniformRangedFloat(Minimum=0.5, Maximum=1.2)
    Mesh Scale=&Multiply_VectorByFloat(Vector=<0.125 0.125 0.01>, Float=&RandomRangeFloat(Minimum=0.75, Maximum=1.5))
    Module.Write Color=false
    Module.Write Lifetime=true
    Module.Write Mass=true
    Module.Write Material Random=true
    Module.Write Position=false
    Module.Write Scale=true
    Module.Write SpriteRotation=false
    Module.Write SpriteSize=true
    Sprite Size=&Vector2DFromFloat(Value=20)
   )
   &AddVelocity(Velocity=&RandomVector(Vector Scale=&RandomRangeFloat(Minimum=20, Maximum=50)))
   &SkeletalMeshLocation(
    Bone Sampling Mode=::NewEnumerator4
    Inherit Velocity=false
    MaskSpawnGroup=false
    Sampled Position Offset=&Multiply_VectorByFloat(Vector=$Particles.Velocity, Float=0.2)
    Skeletal Mesh=$User.SkeletalMesh
   ) {module=SkeletalMeshLocation001}
  ]
  particleUpdateScript [
   &ParticleState()
   &SkeletalMeshLocation(
    Apply To Particle Position=false
    Bone Sampling Mode=::NewEnumerator4
    Inherit Velocity=false
    MaskSpawnGroup=false
    Skeletal Mesh=$User.SkeletalMesh
   )
   /SetVariables(
    Particles.CompareZ:NiagaraFloat=&OneMinusFloat(Float=&ClampFloat(Float=&Multiply_Float(
       A=&Subtract_Float(
        A=&MakeFloatFromVector(Vector=Linked() {linkedVariable {name=Particles.Position, type=NiagaraPosition}}, Channel=::NewEnumerator2)
        B=&MakeFloatFromVector(Vector=$Output.SkeletalMeshLocation.SampledPosition, Channel=::NewEnumerator2)
       )
       B=0.01
      )))
    Particles.MeshPosition:Vector3f=&Lerp_Vector(
     A=$Output.SkeletalMeshLocation.SampledPosition
     B=Linked() {linkedVariable {name=Particles.Position, type=NiagaraPosition}}
     Alpha=&Add_Float(A=0.9, B=&Multiply_Float(A=0.1, B=$Particles.NormalizedAge))
    )
   ) {isSetParameters=true}
   &GravityForce()
   &SolveForcesAndVelocity(Clamp Velocity=false, Limit Acceleration=false)
   &CalculateAccurateVelocity()
   /SetVariables(Particles.MeshFacing:Vector3f=&NormalizeVector(Vector To Normalize=&Subtract_Vector(A=$Output.SkeletalMeshLocation.SampledPosition, B=Linked() {linkedVariable {name=Particles.Position, type=NiagaraPosition}}))) {isSetParameters=true, module=SetVariables_#1}
   &OrientMeshToVector(
    Look At Direction:Vector3f=&Multiply_VectorByFloat(Vector=$Particles.MeshFacing, Float=1)
    Override Quat=false
    Reference Axis Vector:Vector3f=<0 0 1>
    UseOptionalSideAxis=false
    UseOptionalUpAxis=false
   )
   &DynamicMaterialParameters(
    Index 2 Param 1=HlslExpression() {hlslExpression=Particles.MeshPosition.x}
    Index 2 Param 2=HlslExpression() {hlslExpression=Particles.MeshPosition.y}
    Index 2 Param 3=HlslExpression() {hlslExpression=Particles.MeshPosition.z}
    Index 2 Param 4=0.5
    Index 3 Param 1=&MakeFloatFromVector(Vector=Linked() {linkedVariable {name=Particles.Position, type=NiagaraPosition}}, Channel=::NewEnumerator0)
    Index 3 Param 2=&MakeFloatFromVector(Vector=Linked() {linkedVariable {name=Particles.Position, type=NiagaraPosition}}, Channel=::NewEnumerator1)
    Index 3 Param 3=&MakeFloatFromVector(Vector=Linked() {linkedVariable {name=Particles.Position, type=NiagaraPosition}}, Channel=::NewEnumerator2)
    Index 3 Param 4=0.2
    Module.Param0WriteEnabled2=true
    Module.Param0WriteEnabled3=true
    Module.Param1WriteEnabled2=true
    Module.Param1WriteEnabled3=true
    Module.Param2WriteEnabled2=true
    Module.Param2WriteEnabled3=true
    Module.Param3WriteEnabled2=true
    Module.Param3WriteEnabled3=true
    Write Parameter Index 0=false
    Write Parameter Index 2=true
    Write Parameter Index 3=true
   )
   /SetVariables(Particles.XYScale:NiagaraFloat=&FloatFromCurve(FloatCurve=DataInterface() {propertyValues {Curve {keys [
         {arriveTangent=-3.383, interpMode=RCIM_Cubic, leaveTangent=-3.383, tangentMode=RCTM_User, time=0, value=1}
         {arriveTangent=-0.5412, interpMode=RCIM_Cubic, leaveTangent=-0.5412, tangentMode=RCTM_User, time=1, value=0}
        ]}}})) {isSetParameters=true}
   &ScaleMeshSize(Scale Factor=&Multiply_VectorByFloat(
     Vector=&MakeVector(
      X=$Particles.XYScale
      Y=$Particles.XYScale
      Z=&ClampFloat(
       Float=&VectorLength(Vector=&Subtract_Vector(A=$Particles.MeshPosition, B=Linked() {linkedVariable {name=Particles.Position, type=NiagaraPosition}}))
       Max=$Particles.PendulumLength
      )
     )
     Float=$Particles.CompareZ
    )) {module=ScaleMeshSize001}
   &Color(Color=&ColorFromCurve(
     ColorCurve=DataInterface() {propertyValues {
       AlphaCurve {keys [<at 0 20 RCIM_Cubic>]}
       BlueCurve {keys [<at 0 0 RCIM_Cubic>,<at 0.5257 0>,<at 1 5 RCIM_Cubic>]}
       GreenCurve {keys [<at 0 0 RCIM_Cubic>,<at 0.5257 5>,<at 1 0 RCIM_Cubic>]}
       RedCurve {keys [<at 0 5 RCIM_Cubic>,<at 0.5257 0>,<at 1 0 RCIM_Cubic>]}
      }}
     CurveIndex=&VectorLength(Vector=$Particles. PositionCompare)
    )) {enabled=false}
  ]
  rendererClasses [&NiagaraSpriteRendererProperties, &NiagaraMeshRendererProperties]
  renderers [
   {rendererClass=&NiagaraSpriteRendererProperties}
   {rendererClass=&NiagaraMeshRendererProperties, rendererIndex=1}
   {rendererClass=&NiagaraSpriteRendererProperties, rendererIndex=2}
  ]
 }
}
summary {
 systemName=NS_Kepler_Skin
 emitters [
  {emitterName=JumpDrool, rendererClasses [&NiagaraSpriteRendererProperties, &NiagaraMeshRendererProperties]}
  {emitterName=DroolDrops, rendererClasses [&NiagaraSpriteRendererProperties, &NiagaraMeshRendererProperties]}
  {emitterName=DroolPendulum, rendererClasses [&NiagaraSpriteRendererProperties, &NiagaraMeshRendererProperties]}
 ]
 userVariables [
  User.bIsFrontEndPreview {type=NiagaraBool}
  User.ForwardDot {type=NiagaraFloat, defaultValue=0, description=""}
  User.IsSkydiving {type=NiagaraInt32}
  User.JumpTime {type=NiagaraFloat, defaultValue=0, description=""}
  User.ReactiveRate {type=NiagaraFloat, defaultValue=0, description=""}
  User.SkeletalMesh {
   type=NiagaraDataInterfaceSkeletalMesh
   defaultValue {
    struct=NiagaraExt_VariableValue_DataInterface
    value {dataInterface=&self:NiagaraDataInterfaceSkeletalMesh_0}
   }
  }
  User.SpawnBurst {type=NiagaraBool}
 ]
}
