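# Control Rig for the male medium Kepler player character that procedurally animates the inner-mouth bone chain.
# A chain harmonics unit adds pendulum, reach and wave motion from dyn_innerMouth_01, with debug drawing on.
# The jaw's X rotation relative to faceAttach is remapped to weight an additive local bone modifier.
# As the jaw rotates past a threshold, that modifier pushes the chain along local X and rotates the upper and lower inner-mouth bones.
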ControlRigBlueprint /Game/Characters/Player/Male/Medium/Bodies/M_MED_Kepler/Meshes/Parts/M_MED_Kepler_CtrlRig/M_MED_Kepler_CtrlRig
graphs [
 RigVMFunctionLibrary {links [], nodes []}
 RigVMModel {nodes [
   n1 = &RigUnit_BeginExecution()
   /Script/ControlRig.RigUnit_ChainHarmonics(
    ChainRoot=dyn_innerMouth_01
    DrawWorldOffset=(Rotation=(X=0,Y=0,Z=0,W=1),Translation=(X=0,Y=0,Z=0),Scale3D=(X=1,Y=1,Z=1))
    Pendulum=(bEnabled=True,PendulumStiffness=2,PendulumGravity=(X=0,Y=0,Z=0),PendulumBlend=0.75,PendulumDrag=0.98,PendulumMaximum=0.1,UnwindAxis=(X=0,Y=1,Z=0),UnwindMinimum=0.2,UnwindMaximum=0.05)
    Reach=(bEnabled=True,ReachTarget=(X=0,Y=0,Z=0),ReachAxis=(X=1,Y=0,Z=0))
    Speed=(X=1,Y=1,Z=1)
    Wave=(bEnabled=True,WaveFrequency=(X=0.4,Y=0.5,Z=0.7),WaveAmplitude=(X=0.06,Y=0.05,Z=0.04),WaveOffset=(X=0,Y=1,Z=2),WaveNoise=(X=0,Y=0,Z=0),WaveMaximum=1)
    WaveCurve=(EditorCurveData=(Keys=((),(Time=1,Value=1)),DefaultValue=340282346638528859811704183484516925440,PreInfinityExtrap=RCCE_Constant,PostInfinityExtrap=RCCE_Constant))
    bDrawDebug=True
    ExecutePin=@n1.ExecutePin
   ) {is=struct RigVMUnitNode}
   n2 = &RigUnit_GetRelativeBoneTransform(Bone=C_jaw, Space=faceAttach)
   > /Script/ControlRig.RigUnit_ModifyBoneTransforms(
    BoneToModify="((Bone=\"dyn_innerMouth_01\",Transform=(Rotation=(X=0,Y=0,Z=0,W=1),Translation=(X=6,Y=0,Z=0),Scale3D=(X=1,Y=1,Z=1))),(Bone=\"dyn_innerMouth_upper\",Transform=(Rotation=(X=0,Y=0,Z=0,W=1),Translation=(X=0,Y=0,Z=0),Scale3D=(X=1,Y=1,Z=1))),(Bone=\"dyn_innerMouth_lo ...(truncated)"
    Mode=AdditiveLocal
    WeightMaximum=1
    Weight=&RigVMFunction_MathFloatRemap(
     SourceMaximum=-30
     SourceMinimum=-20
     TargetMinimum=0
     Value=&RigVMFunction_MathQuaternionToEuler(RotationOrder=XYZ, Value=@n2.Transform.Rotation).Result.X
    ).Result
    BoneToModify.1.Transform.Rotation=&RigVMFunction_MathQuaternionFromEuler(Euler=(X=0,Y=-50,Z=0)).Result
    BoneToModify.2.Transform.Rotation=&RigVMFunction_MathQuaternionFromEuler(Euler=(X=0,Y=-50,Z=0)).Result
   ) {is=struct RigVMUnitNode}
  ]}
]
hierarchy [
 root {loc=<0 0 0>, rot=<0 0 0 1>}
 pelvis {parent=root, loc=<0 -0.7071 93.1191>, rot=<0.0127 0.707 -0.0127 -0.707>}
 spine_01 {parent=pelvis, loc=<6.7494 -0.6388 0>, rot=<0 0 0.034 -0.9994>}
 spine_02 {parent=spine_01, loc=<7.2518 -0.035 0>, rot=<0 0 0.0238 -0.9997>}
 spine_03 {parent=spine_02, loc=<6.0104 0 0>, rot=<0 0 0.0149 -0.9999>}
 spine_04 {parent=spine_03, loc=<8.1024 0 0>, rot=<0 0 -0.0132 -0.9999>}
 spine_05 {parent=spine_04, loc=<10.0146 0 0>, rot=<0 0 -0.0015 -1>}
 clavicle_l {parent=spine_05, loc=<4.3575 -0.1466 -2.3223>, rot=<0.0465 -0.7289 -0.009 -0.6829>}
 upperarm_l {parent=clavicle_l, loc=<15.5111 0 0>, rot=<0.0226 -0.3234 -0.0088 -0.9459>}
 upperarm_twist_01_l {parent=upperarm_l, loc=<0.0041 0 0>, rot=<0 0 0 1>}
 clavicle_r {parent=spine_05, loc=<4.3575 -0.1466 2.3223>, rot=<0.7289 0.0465 -0.6829 0.009>}
 upperarm_r {parent=clavicle_r, loc=<-15.5111 0 0>, rot=<0.0226 -0.3234 -0.0088 -0.9459>}
 upperarm_twist_01_r {parent=upperarm_r, loc=<-0.0041 0 0>, rot=<0 0 0 1>}
 neck_01 {parent=spine_05, loc=<11.3138 0.1529 0>, rot=<0 0 -0.2003 -0.9797>}
 neck_02 {parent=neck_01, loc=<4.5627 0 0>, rot=<0 0 0.0504 -0.9987>}
 head {parent=neck_02, loc=<5.3914 0 0>, rot=<0 0 0.115 -0.9934>}
 faceAttach {parent=head, loc=<-4.1329 -0.9158 0>, rot=<0.0082 -0.7071 -0.0082 -0.7071>, scale=<0.89 0.89 0.89>}
 L_cheek_inner {parent=faceAttach, loc=<4.9398 12.2161 3.6809>, rot=<0.0742 0.0933 0.5873 -0.8006>}
 C_lip_upper_mid {parent=faceAttach, loc=<0.0023 21.3886 4.0585>, rot=<0.0987 0.0013 -0.0007 -0.9951>}
 L_lip_upper_outer {parent=faceAttach, loc=<2.7719 20.1029 4.1922>, rot=<0.05 0.0205 0.4651 -0.8836>}
 L_lip_corner {parent=faceAttach, loc=<4.0727 15.2625 2.845>, rot=<0.04 -0.0723 -0.1154 -0.9899>}
 C_nose_bridge {parent=faceAttach, loc=<-0.0198 21.6123 6.7615>, rot=<-0.099 0 0 -0.9951>}
 R_cheek_inner {parent=faceAttach, loc=<-4.9399 12.2161 3.6809>, rot=<0.0742 -0.0933 -0.5873 -0.8006>}
 R_lip_corner {parent=faceAttach, loc=<-4.0727 15.2625 2.845>, rot=<0.04 0.0723 0.1154 -0.9899>}
 C_jaw {parent=faceAttach, loc=<-0.0169 10.1147 4.6386>, rot=<0 0 0 1>}
 L_lip_lower_outer {parent=C_jaw, loc=<2.4763 9.3598 -4.0146>, rot=<0.0046 0.0461 0.4751 -0.8787>}
 C_lip_lower_mid {parent=C_jaw, loc=<0.0269 10.8385 -4.136>, rot=<0 0 0 -1>}
 R_lip_lower_outer {parent=C_jaw, loc=<-2.4424 9.3598 -4.0146>, rot=<0.0046 -0.0461 -0.4751 -0.8787>}
 teeth_lower {parent=C_jaw, loc=<0.0169 10.3801 -5.2573>, rot=<0.1032 0 0 -0.9947>}
 teeth_upper {parent=faceAttach, loc=<0 19.1831 5.8231>, rot=<0.118 0 0 -0.993>}
 R_lip_upper_outer {parent=faceAttach, loc=<-2.7719 20.1029 4.1922>, rot=<0.05 -0.0205 -0.4651 -0.8836>}
 dyn_innerMouth_mstr {parent=head, loc=<-1.6575 3.261 0>, rot=<0.4905 0.5093 0.5093 0.4905>}
 dyn_innerMouth_01 {parent=dyn_innerMouth_mstr, loc=<0 0 0>, rot=<0 0 0 1>}
 dyn_innerMouth_02 {parent=dyn_innerMouth_01, loc=<4.1371 0 0>, rot=<0 0 0.0025 -1>}
 dyn_innerMouth_03 {parent=dyn_innerMouth_02, loc=<4.2915 0 0>, rot=<0 0 0 1>}
 dyn_innerMouth_upper {parent=dyn_innerMouth_03, loc=<4.3527 0.0435 1.1641>, rot=<0 0.0146 -0.0025 -0.9999>}
 dyn_innerMouth_lower {parent=dyn_innerMouth_03, loc=<4.2834 0.0432 -1.841>, rot=<0.9999 0.0025 0.0146 0>}
]
