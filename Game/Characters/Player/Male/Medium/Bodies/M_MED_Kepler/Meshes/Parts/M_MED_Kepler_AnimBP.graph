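# Animation Blueprint for a mesh part of the male Kepler player character, driving its pose and an emote-specific material toggle.
# The AnimGraph copies the attached parent's pose, constrains the inner-mouth bone to the jaw at half alpha, then runs the Kepler Control Rig.
# Each update, the EventGraph sets the part material's "Hide Element 01" parameter on while the owner mesh plays the Kepler emote montage, otherwise off.
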AnimBlueprint /Game/Characters/Player/Male/Medium/Bodies/M_MED_Kepler/Meshes/Parts/M_MED_Kepler_AnimBP
target_skeleton=&/M_MED_Kepler_Skeleton
graphs [
 AnimGraph {nodes [
   n1 = "Constraint - dyn_innerMouth_mstr (P)"(ComponentPose=Local To Component(LocalPose=Copy Pose From Mesh() {bUseAttachedParent=true}), Alpha=0.5, ConstraintWeights_0=1) {
    alphaBoolBlend {blendOption=Linear}
    boneToModify {boneName=dyn_innerMouth_mstr}
    constraintSetup [{targetBone {boneName=C_jaw}}]
    constraintWeights [1]
   }
   n2 = Control Rig(Source=Component To Local(ComponentPose=@n1)) {
    alphaBoolBlend {blendOption=Linear}
    controlRigAssetReference {blueprintRigClass=/Game/Characters/Player/Male/Medium/Bodies/M_MED_Kepler/Meshes/Parts/M_MED_Kepler_CtrlRig/M_MED_Kepler_CtrlRig.M_MED_Kepler_CtrlRig_C}
   }
   Output Pose(Result=@n2) {layerGroup=None}
  ]}
 EventGraph {nodes [
   Event BlueprintUpdateAnimation()
   n1 = GetAnimInstance(self=GetOwnerSkeletalMeshComponent())
   > Is Valid(InputObject=@n1)
   > n2 = Get Dynamic Material Instance(SkeletalMeshComponent=GetPartSkeletalMeshComponent())
   > SetScalarParameterValue(
    self=@n2
    ParameterName=Hide Element 01
    Value=SelectFloat(
     A=1
     B=0
     bPickA=Montage_IsActive!(
      self:Anim Instance Object Reference=@n1
      Montage:Anim Montage Object Reference=/Game/Animation/Game/MainPlayer/Emotes/Kepler/Emote_KeplerMale_CMM_M
     ) -> (ReturnValue:bool)
    )
   )
  ]}
]
variables []
